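# Control Rig for the HarborSnow large male head that adds procedural neck twist and corrective deformation as the head moves.
# A sequence averages rear and side deform roots between head and neck_twist, then feeds partial head twist into two neck bones.
# Spherical pose readers on the head, relative to spine_05, then weight offsets on those deform bones.
# The isEmoting head-rotation remap and jaw-chin corrective are disconnected from execution.
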ControlRigBlueprint /BRCosmetics/Characters/Player/Male/Large/Heads/M_LRG_HarborSnow_Head/Meshes/M_LRG_HarborSnow_Head_CtrlRig
graphs [
 RigVMFunctionLibrary {links [], nodes []}
 RigVMFunction_Sequence_1_ContainedGraph {nodes [
   n1 = RigVMFunctionEntryNode()
   RigVMFunction_Sequence = &RigVMFunction_Sequence(ExecuteContext=@n1.ExecuteContext)
   n2 = &RigVMFunction_Sequence(ExecuteContext=@RigVMFunction_Sequence.B)
   n3 = &RigVMFunction_Sequence(ExecuteContext=@n2.B)
   RigVMFunctionReturnNode(A=@RigVMFunction_Sequence.A, B=@n2.A, C=@n3.A, D=@n3.B)
  ]}
 RigVMModel {nodes [
   RigUnit_BeginExecution = &RigUnit_BeginExecution()
   RigVMFunction_Sequence = RigVMAggregateNode(ExecuteContext=@RigUnit_BeginExecution.ExecutePin)
   &RigUnit_ParentConstraint(
    AdvancedSettings=(InterpolationType=Average,RotationOrderForFilter=XZY)
    Child=(Type=Bone,Name="deform_head_rear_root")
    Parents=((Item=(Type=Bone,Name="head"),Weight=0.5),(Item=(Type=Bone,Name="neck_twist"),Weight=0.5))
    Weight=1
    ExecutePin=@RigVMFunction_Sequence.A
   )
   > &RigUnit_ParentConstraint(
    AdvancedSettings=(InterpolationType=Average,RotationOrderForFilter=XZY)
    Child=(Type=Bone,Name="deform_head_l_root")
    Parents=((Item=(Type=Bone,Name="head"),Weight=0.5),(Item=(Type=Bone,Name="neck_twist"),Weight=0.5))
    Weight=1
   )
   > &RigUnit_ParentConstraint(
    AdvancedSettings=(InterpolationType=Average,RotationOrderForFilter=XZY)
    Child=(Type=Bone,Name="deform_head_r_root")
    Parents=((Item=(Type=Bone,Name="head"),Weight=0.5),(Item=(Type=Bone,Name="neck_twist"),Weight=0.5))
    Weight=1
   )
   n1 = &RigUnit_SphericalPoseReader(
    ActiveRegionScaleFactors=(PositiveWidth=1,NegativeWidth=3,PositiveHeight=3.04,NegativeHeight=3)
    ActiveRegionSize=0.1
    Debug=(bDrawDebug=true,bDraw2D=False,bDrawLocalAxes=False,DebugScale=25,DebugSegments=20,DebugThickness=0.25)
    DriverAxis=(X=1,Y=0,Z=0)
    DriverItem=(Type=Bone,Name="head")
    FalloffRegionScaleFactors=(PositiveWidth=1,NegativeWidth=1,PositiveHeight=1,NegativeHeight=1)
    FalloffSize=0.3
    FlipHeightScaling=False
    FlipWidthScaling=False
    OptionalParentItem=(Type=Bone,Name="spine_05")
    RotationOffset=(X=-90,Y=0,Z=20)
    ExecutePin=@RigVMFunction_Sequence.C
   )
   > &RigUnit_ModifyTransforms(
    ItemToModify=((Item=(Type=Bone,Name="deform_head_rear"),Transform=(Rotation=(X=0,Y=0,Z=0.0872,W=0.9962),Translation=(X=3,Y=-10,Z=0),Scale3D=(X=1,Y=1,Z=1))))
    WeightMinimum=0
    Weight=@n1.OutputParam
   )
   > n2 = &RigUnit_SphericalPoseReader(
    ActiveRegionScaleFactors=(PositiveWidth=1,NegativeWidth=1,PositiveHeight=1,NegativeHeight=1)
    ActiveRegionSize=0.075
    Debug=(bDrawDebug=true,bDraw2D=False,bDrawLocalAxes=False,DebugScale=25,DebugSegments=20,DebugThickness=0.25)
    DriverAxis=(X=1,Y=0,Z=0)
    DriverItem=(Type=Bone,Name="head")
    FalloffRegionScaleFactors=(PositiveWidth=1,NegativeWidth=1,PositiveHeight=1,NegativeHeight=1)
    FalloffSize=0.25
    FlipHeightScaling=False
    FlipWidthScaling=False
    OptionalParentItem=(Type=Bone,Name="spine_05")
    RotationOffset=(X=0,Y=-45,Z=0)
   )
   > &RigUnit_ModifyTransforms(
    ItemToModify=((Item=(Type=Bone,Name="deform_head_r"),Transform=(Rotation=(X=0,Y=0,Z=0.0872,W=0.9962),Translation=(X=-5,Y=0,Z=7),Scale3D=(X=1.2,Y=1.2,Z=1.2))))
    WeightMinimum=0
    Weight=@n2.OutputParam
   )
   > n3 = &RigUnit_SphericalPoseReader(
    ActiveRegionScaleFactors=(PositiveWidth=1,NegativeWidth=1,PositiveHeight=1,NegativeHeight=1)
    ActiveRegionSize=0.075
    Debug=(bDrawDebug=true,bDraw2D=False,bDrawLocalAxes=False,DebugScale=25,DebugSegments=20,DebugThickness=0.25)
    DriverAxis=(X=1,Y=0,Z=0)
    DriverItem=(Type=Bone,Name="head")
    FalloffRegionScaleFactors=(PositiveWidth=1,NegativeWidth=1,PositiveHeight=1,NegativeHeight=1)
    FalloffSize=0.25
    FlipHeightScaling=False
    FlipWidthScaling=False
    OptionalParentItem=(Type=Bone,Name="spine_05")
    RotationOffset=(X=0,Y=225,Z=0)
   )
   > &RigUnit_ModifyTransforms(
    ItemToModify=((Item=(Type=Bone,Name="deform_head_l"),Transform=(Rotation=(X=0,Y=0,Z=0.0872,W=0.9962),Translation=(X=5,Y=0,Z=-7),Scale3D=(X=1.2,Y=1.2,Z=1.2))))
    WeightMinimum=0
    Weight=@n3.OutputParam
   )
   > n4 = &RigUnit_SphericalPoseReader(
    ActiveRegionScaleFactors=(PositiveWidth=1,NegativeWidth=1,PositiveHeight=1,NegativeHeight=1)
    ActiveRegionSize=0.1
    Debug=(bDrawDebug=true,bDraw2D=False,bDrawLocalAxes=False,DebugScale=25,DebugSegments=20,DebugThickness=0.25)
    DriverAxis=(X=0,Y=-1,Z=0)
    DriverItem=(Type=Bone,Name="head")
    FalloffRegionScaleFactors=(PositiveWidth=1,NegativeWidth=1,PositiveHeight=1,NegativeHeight=1)
    FalloffSize=0.3
    FlipHeightScaling=False
    FlipWidthScaling=False
    OptionalParentItem=(Type=Bone,Name="spine_05")
    RotationOffset=(X=20,Y=-10,Z=20)
   )
   > &RigUnit_ModifyTransforms(
    ItemToModify="((Item=(Type=Bone,Name=\"deform_head_rear\"),Transform=(Rotation=(X=0,Y=0,Z=0,W=1),Translation=(X=0,Y=-10,Z=0),Scale3D=(X=1,Y=1,Z=1))),(Item=(Type=Bone,Name=\"deform_head_r\"),Transform=(Rotation=(X=0,Y=0,Z=0,W=1),Translation=(X=0,Y=0,Z=0),Scale3D=(X=1,Y=1,Z=1 ...(truncated)"
    WeightMinimum=0
    Weight=@n4.OutputParam
   )
   > n5 = &RigUnit_SphericalPoseReader(
    ActiveRegionScaleFactors=(PositiveWidth=1,NegativeWidth=1,PositiveHeight=1,NegativeHeight=1)
    ActiveRegionSize=0.1
    Debug=(bDrawDebug=true,bDraw2D=False,bDrawLocalAxes=False,DebugScale=25,DebugSegments=20,DebugThickness=0.25)
    DriverAxis=(X=0,Y=-1,Z=0)
    DriverItem=(Type=Bone,Name="head")
    FalloffRegionScaleFactors=(PositiveWidth=1,NegativeWidth=1,PositiveHeight=1,NegativeHeight=1)
    FalloffSize=0.3
    FlipHeightScaling=false
    FlipWidthScaling=false
    OptionalParentItem=(Type=Bone,Name="spine_05")
    RotationOffset=(X=190,Y=-10,Z=20)
   )
   > &RigUnit_ModifyTransforms(
    ItemToModify="((Item=(Type=Bone,Name=\"deform_head_rear\"),Transform=(Rotation=(X=0,Y=0,Z=0,W=1),Translation=(X=0,Y=-10,Z=0),Scale3D=(X=1,Y=1,Z=1))),(Item=(Type=Bone,Name=\"deform_head_r\"),Transform=(Rotation=(X=0,Y=0,Z=0,W=1),Translation=(X=0,Y=0,Z=0),Scale3D=(X=1,Y=1,Z=1 ...(truncated)"
    WeightMinimum=0
    Weight=@n5.OutputParam
   )
   RigUnit_GetTransform = &RigUnit_GetTransform(Item=(Type=Bone,Name="head"), Space=LocalSpace, bInitial=False)
   RigVMFunction_MathQuaternionSwingTwist = &RigVMFunction_MathQuaternionSwingTwist(TwistAxis=(X=1,Y=0,Z=0), Input=@RigUnit_GetTransform.Transform.Rotation)
   Interpolate = /Script/RigVM.RigVMFunction_MathQuaternionSlerp(A=(X=0,Y=0,Z=0,W=1), T=0.5, B=@RigVMFunction_MathQuaternionSwingTwist.Twist) {is=struct RigVMUnitNode, template="Interpolate::Execute(in A,in B,in T,out Result)"}
   &RigUnit_ModifyTransforms(
    ItemToModify=((Item=(Type=Bone,Name="neck_twist"),Transform=(Rotation=(X=0,Y=0,Z=0,W=1),Translation=(X=0,Y=0,Z=0),Scale3D=(X=1,Y=1,Z=1))))
    Weight=1
    WeightMinimum=0
    ItemToModify.0.Transform.Rotation=@Interpolate.Result
    ExecutePin=@RigVMFunction_Sequence.B
   )
   n6 = &RigUnit_GetTransform(Item=(Type=Bone,Name="head"), Space=LocalSpace, bInitial=False)
   > &RigUnit_ModifyTransforms(
    ItemToModify=((Item=(Type=Bone,Name="neck_twist_lower"),Transform=(Rotation=(X=0,Y=0,Z=0,W=1),Translation=(X=0,Y=0,Z=0),Scale3D=(X=1,Y=1,Z=1))))
    Weight=1
    WeightMinimum=0
    ItemToModify.0.Transform.Rotation=&RigVMFunction_MathQuaternionSlerp(
     A=(X=0,Y=0,Z=0,W=1)
     T=0.2
     B=&RigVMFunction_MathQuaternionSwingTwist(TwistAxis=(X=1,Y=0,Z=0), Input=@n6.Transform.Rotation).Twist
    ) {template="Interpolate::Execute(in A,in B,in T,out Result)"}.Result
   )
   RigVMFunction_ControlFlowBranch = &RigVMFunction_ControlFlowBranch(ExecuteContext=(), Condition=RigVMVariableNode() {variable=isEmoting}.Value)
   RigVMFunction_MathQuaternionToEuler = &RigVMFunction_MathQuaternionToEuler(
    RotationOrder=ZYX
    Value=&RigUnit_GetTransform(Item=(Type=Bone,Name="head"), Space=LocalSpace, bInitial=False).Transform.Rotation
   )
   RigVMFunction_AnimEvalRichCurve = &RigVMFunction_AnimEvalRichCurve(
    Curve=(EditorCurveData=(Keys=((),(InterpMode=RCIM_Cubic,Time=0.6897,Value=0.8554,ArriveTangent=1,LeaveTangent=1),(Time=1,Value=1))))
    SourceMaximum=-45
    SourceMinimum=10
    TargetMaximum=-20
    TargetMinimum=10
    Value=@RigVMFunction_MathQuaternionToEuler.Result.Z
   )
   RigVMFunction_MathQuaternionFromEuler = &RigVMFunction_MathQuaternionFromEuler(
    Euler=(X=0,Y=0,Z=0)
    Euler.Z=@RigVMFunction_AnimEvalRichCurve.Result
    Euler.X=@RigVMFunction_MathQuaternionToEuler.Result.X
    Euler.Y=@RigVMFunction_MathQuaternionToEuler.Result.Y
   )
   RigUnit_SetRotation = &RigUnit_SetRotation(
    Item=(Type=Bone,Name="head")
    Weight=1
    bInitial=False
    Value=@RigVMFunction_MathQuaternionFromEuler.Result
    ExecutePin=@RigVMFunction_ControlFlowBranch.True
   )
   n7 = &RigUnit_SphericalPoseReader(
    ActiveRegionScaleFactors=(PositiveWidth=1,NegativeWidth=1,PositiveHeight=1,NegativeHeight=1)
    ActiveRegionSize=0.1
    Debug=(bDrawDebug=false,bDraw2D=False,bDrawLocalAxes=False,DebugScale=25,DebugSegments=20,DebugThickness=0.25)
    DriverAxis=(X=0,Y=1,Z=0)
    DriverItem=(Type=Bone,Name="C_jaw")
    ExecutePin=()
    FalloffRegionScaleFactors=(PositiveWidth=1,NegativeWidth=1,PositiveHeight=1,NegativeHeight=1)
    FalloffSize=0.2
    FlipHeightScaling=False
    FlipWidthScaling=False
    OptionalParentItem=(Type=Bone,Name="head")
    RotationOffset=(X=130,Y=0,Z=0)
   )
   &RigUnit_ModifyTransforms(
    ExecutePin=()
    ItemToModify=((Item=(Type=Bone,Name="C_jaw_chin_deform"),Transform=(Rotation=(X=0.1874,Y=0,Z=0,W=0.9823),Translation=(X=0,Y=3.3337,Z=0.7017),Scale3D=(X=1,Y=0.875,Z=1))))
    WeightMinimum=0
    Weight=@n7.OutputParam
   )
  ]}
]
hierarchy [
 root {loc=<0 0 0>, rot=<0 0 0 1>}
 pelvis {parent=root, loc=<0 -0.7321 92.4218>, rot=<0.0127 0.707 -0.0127 -0.707>}
 spine_01 {parent=pelvis, loc=<5.6002 -0.5128 0>, rot=<0 0 0.034 -0.9994>}
 spine_02 {parent=spine_01, loc=<8.0756 0.0139 0>, rot=<0 0 0.0238 -0.9997>}
 spine_03 {parent=spine_02, loc=<7.0395 0 0>, rot=<0 0 0.0149 -0.9999>}
 spine_04 {parent=spine_03, loc=<8.1023 0 0>, rot=<0 0 -0.0132 -0.9999>}
 spine_05 {parent=spine_04, loc=<10.0141 0 0>, rot=<0 0 -0.0015 -1>}
 clavicle_l {parent=spine_05, loc=<2.5856 -0.2866 -4.3777>, rot=<0.0465 -0.7289 -0.009 -0.6829>}
 upperarm_l {parent=clavicle_l, loc=<18.2264 0.1342 -0.001>, rot=<0.0253 -0.2693 0.0013 -0.9627>}
 upperarm_twist_01_l {parent=upperarm_l, loc=<7.9796 0.0843 -0.6195>, rot=<0 0 0 -1>}
 neck_01 {parent=spine_05, loc=<10.815 1.9196 0>, rot=<0 0 -0.1864 -0.9825>}
 neck_02 {parent=neck_01, loc=<4.7347 0.3045 0>, rot=<0 0 0.0363 -0.9993>}
 head {parent=neck_02, loc=<5.7632 0.5552 0>, rot=<0 0 0.115 -0.9934>}
 faceAttach {parent=head, loc=<-4.146 -0.8545 0>, rot=<0.0029 -0.7071 -0.0029 -0.7071>}
 C_lip_upper_mid {parent=faceAttach, loc=<0.0022 16.8076 1.148>, rot=<0.2225 0.0013 -0.0005 -0.9749>}
 L_lip_upper_outer {parent=faceAttach, loc=<8.8441 14.829 1.3354>, rot=<0.0359 -0.0728 0.2627 -0.9615>}
 L_lip_corner {parent=faceAttach, loc=<14.176 12.1991 1.2088>, rot=<0.003 0.008 -0.0146 -0.9999>}
 R_lip_corner {parent=faceAttach, loc=<-14.176 12.1991 1.2088>, rot=<0.003 -0.008 0.0146 -0.9999>}
 C_jaw {parent=faceAttach, loc=<-0.0134 -0.1613 -1.4291>, rot=<0 0 0 1>}
 L_lip_lower_outer {parent=C_jaw, loc=<8.8405 14.8022 1.4143>, rot=<0.013 0.0445 0.3057 -0.951>}
 C_lip_lower_mid {parent=C_jaw, loc=<0.0223 16.7112 1.2233>, rot=<0 0 0 -1>}
 R_lip_lower_outer {parent=C_jaw, loc=<-8.8137 14.8022 1.4143>, rot=<0.013 -0.0445 -0.3057 -0.951>}
 teeth_lower {parent=C_jaw, loc=<0.0134 6.0835 -1.7637>, rot=<0 0 0 1>}
 tongue {parent=C_jaw, loc=<0.0134 -8.8951 -2.0419>, rot=<-0.1568 0 0 -0.9876>}
 R_eye {parent=faceAttach, loc=<-9.0377 14.5612 6.8224>, rot=<0 0 0.9998 -0.0175>}
 R_eye_lid_upper_mid {parent=R_eye, loc=<0 0 0>, rot=<0.0523 0 0 -0.9986>}
 R_eye_lid_lower_mid {parent=R_eye, loc=<0 0 0>, rot=<0.0329 0 0 -0.9995>}
 L_eye {parent=faceAttach, loc=<9.0377 14.5613 6.8224>, rot=<0 0 0.9998 0.0175>}
 L_eye_lid_upper_mid {parent=L_eye, loc=<0 0 0>, rot=<0.0523 0 0 -0.9986>}
 L_eye_lid_lower_mid {parent=L_eye, loc=<0 0 0>, rot=<0.0329 0 0 -0.9995>}
 teeth_upper {parent=faceAttach, loc=<0 6.5028 4.229>, rot=<0 0 0 1>}
 R_lip_upper_outer {parent=faceAttach, loc=<-8.8441 14.829 1.3354>, rot=<0.0359 0.0728 -0.2627 -0.9615>}
 dyn_hair_01 {parent=head, loc=<13.6636 6.6968 1.3731>, rot=<0.0992 -0.4448 -0.2073 -0.8656>}
 dyn_hair_02 {parent=dyn_hair_01, loc=<4.1335 0 0>, rot=<-0.0054 0.062 0.0283 -0.9977>}
 dyn_hair_03 {parent=dyn_hair_02, loc=<3.2479 0 0>, rot=<0.0011 0.037 0.0236 -0.999>}
 dyn_ear_l_01 {parent=head, loc=<11.412 -2.5996 -13.5489>, rot=<0.692 -0.1574 -0.0658 -0.7014>}
 dyn_ear_l_02 {parent=dyn_ear_l_01, loc=<4.9024 0 0>, rot=<-0.0073 -0.1525 0.0473 -0.9871>}
 dyn_ear_l_03 {parent=dyn_ear_l_02, loc=<4.6336 0 0>, rot=<0.0007 -0.1229 -0.006 -0.9924>}
 dyn_ear_r_01 {parent=head, loc=<11.412 -2.5996 13.5489>, rot=<0.7536 0.0532 0.1621 -0.6349>}
 dyn_ear_r_02 {parent=dyn_ear_r_01, loc=<4.9024 0 0>, rot=<0.004 -0.1434 -0.0701 -0.9872>}
 dyn_ear_r_03 {parent=dyn_ear_r_02, loc=<4.6336 0 0>, rot=<0.0023 -0.0907 -0.0256 -0.9955>}
 deform_head_rear_root {parent=head, loc=<-1.1977 -20.4696 0>, rot=<0 0 0 1>}
 deform_head_rear {parent=deform_head_rear_root, loc=<0 0 0>, rot=<0 0 0 1>}
 deform_head_l_root {parent=head, loc=<-1.5355 -4.591 -20.4092>, rot=<0.0021 -0.0099 -0.0008 -0.9999>}
 deform_head_l {parent=deform_head_l_root, loc=<0 0 0>, rot=<0 0 0 1>}
 deform_head_r_root {parent=head, loc=<-1.5353 -4.591 20.4092>, rot=<0 0 0 1>}
 deform_head_r {parent=deform_head_r_root, loc=<0 0 0>, rot=<0 0 0 1>}
 FX_EyeTrail_L {parent=head, type=Null, loc=<4.75 8 2.6>, rot=<-0.0051 0.6948 0.0011 0.7192>}
 FX_EyeTrail_R {parent=head, type=Null, loc=<4.75 8 -2.831>, rot=<0 0.7071 0 0.7071>}
 FX_Eye_L {parent=head, type=Null, loc=<3.4814 5.9142 -2.8258>, rot=<-0.0059 0.7071 0 0.7071>}
 FX_Eye_R {parent=head, type=Null, loc=<3.4814 5.9142 2.8258>, rot=<-0.0059 0.7071 0 0.7071>}
 earpiece {parent=head, type=Null, loc=<0 0 0>, rot=<0 0 0 1>}
 hat {parent=head, type=Null, loc=<0 0 0>, rot=<0 0.7071 0 0.7071>}
 FX_Visor {parent=head, type=Null, loc=<1.6781 9.0813 0>, rot=<0 0 0 1>}
 neck_twist {parent=neck_01, loc=<4.7347 0.3045 0>, rot=<0 0 0.0363 -0.9993>}
 clavicle_r {parent=spine_05, loc=<2.5855 -0.2866 4.3777>, rot=<0.7289 0.0465 -0.6829 0.009>}
 upperarm_r {parent=clavicle_r, loc=<-18.2264 -0.1342 0.001>, rot=<0.0253 -0.2693 0.0013 -0.9627>}
 upperarm_twist_01_r {parent=upperarm_r, loc=<-7.9796 -0.0843 0.6195>, rot=<0 0 0 -1>}
 neck_twist_lower {parent=spine_05, loc=<10.815 1.9196 0>, rot=<0 0 -0.1506 -0.9886>}
 Charm_Attach_00 {parent=spine_05, type=Null, loc=<-0.095 14.4 -7.75>, rot=<-0.5245 0.4379 -0.6186 0.388>}
 look_left_pose {type=Curve, loc=<0 0 0>, rot=<0 0 0 1>}
 look_right_pose {type=Curve, loc=<0 0 0>, rot=<0 0 0 1>}
 look_up_pose {type=Curve, loc=<0 0 0>, rot=<0 0 0 1>}
 look_down_pose {type=Curve, loc=<0 0 0>, rot=<0 0 0 1>}
 L_blink_pose {type=Curve, loc=<0 0 0>, rot=<0 0 0 1>}
 R_blink_pose {type=Curve, loc=<0 0 0>, rot=<0 0 0 1>}
 L_wide_pose {type=Curve, loc=<0 0 0>, rot=<0 0 0 1>}
 R_wide_pose {type=Curve, loc=<0 0 0>, rot=<0 0 0 1>}
 R_squint_inner_pose {type=Curve, loc=<0 0 0>, rot=<0 0 0 1>}
 R_squeeze_pose {type=Curve, loc=<0 0 0>, rot=<0 0 0 1>}
 L_squint_inner_pose {type=Curve, loc=<0 0 0>, rot=<0 0 0 1>}
 L_squeeze_pose {type=Curve, loc=<0 0 0>, rot=<0 0 0 1>}
 L_brow_up_pose {type=Curve, loc=<0 0 0>, rot=<0 0 0 1>}
 L_brow_down_pose {type=Curve, loc=<0 0 0>, rot=<0 0 0 1>}
 R_brow_up_pose {type=Curve, loc=<0 0 0>, rot=<0 0 0 1>}
 R_brow_down_pose {type=Curve, loc=<0 0 0>, rot=<0 0 0 1>}
 C_glabella_up_pose {type=Curve, loc=<0 0 0>, rot=<0 0 0 1>}
 C_glabella_down_pose {type=Curve, loc=<0 0 0>, rot=<0 0 0 1>}
 L_brow_squeeze_in_pose {type=Curve, loc=<0 0 0>, rot=<0 0 0 1>}
 R_brow_squeeze_in_pose {type=Curve, loc=<0 0 0>, rot=<0 0 0 1>}
 L_upper_lip_raiser_pose {type=Curve, loc=<0 0 0>, rot=<0 0 0 1>}
 R_upper_lip_raiser_pose {type=Curve, loc=<0 0 0>, rot=<0 0 0 1>}
 L_upper_lip_lower_pose {type=Curve, loc=<0 0 0>, rot=<0 0 0 1>}
 R_upper_lip_lower_pose {type=Curve, loc=<0 0 0>, rot=<0 0 0 1>}
 L_lower_lip_up_pose {type=Curve, loc=<0 0 0>, rot=<0 0 0 1>}
 R_lower_lip_up_pose {type=Curve, loc=<0 0 0>, rot=<0 0 0 1>}
 L_lower_lip_down_pose {type=Curve, loc=<0 0 0>, rot=<0 0 0 1>}
 R_lower_lip_down_pose {type=Curve, loc=<0 0 0>, rot=<0 0 0 1>}
 L_lip_corner_narrow_pose {type=Curve, loc=<0 0 0>, rot=<0 0 0 1>}
 L_lip_corner_wide_pose {type=Curve, loc=<0 0 0>, rot=<0 0 0 1>}
 R_lip_corner_narrow_pose {type=Curve, loc=<0 0 0>, rot=<0 0 0 1>}
 R_lip_corner_wide_pose {type=Curve, loc=<0 0 0>, rot=<0 0 0 1>}
 L_smile_pose {type=Curve, loc=<0 0 0>, rot=<0 0 0 1>}
 R_smile_pose {type=Curve, loc=<0 0 0>, rot=<0 0 0 1>}
 L_frown_pose {type=Curve, loc=<0 0 0>, rot=<0 0 0 1>}
 R_frown_pose {type=Curve, loc=<0 0 0>, rot=<0 0 0 1>}
 phoneme_oo_pose {type=Curve, loc=<0 0 0>, rot=<0 0 0 1>}
 phoneme_ch_pose {type=Curve, loc=<0 0 0>, rot=<0 0 0 1>}
 phoneme_mbp_pose {type=Curve, loc=<0 0 0>, rot=<0 0 0 1>}
 phoneme_fv_pose {type=Curve, loc=<0 0 0>, rot=<0 0 0 1>}
 mouth_up_pose {type=Curve, loc=<0 0 0>, rot=<0 0 0 1>}
 mouth_down_pose {type=Curve, loc=<0 0 0>, rot=<0 0 0 1>}
 C_nose_wrinkler_pose {type=Curve, loc=<0 0 0>, rot=<0 0 0 1>}
 L_cheek_up_pose {type=Curve, loc=<0 0 0>, rot=<0 0 0 1>}
 R_cheek_up_pose {type=Curve, loc=<0 0 0>, rot=<0 0 0 1>}
 jaw_open_pose {type=Curve, loc=<0 0 0>, rot=<0 0 0 1>}
 jaw_right_pose {type=Curve, loc=<0 0 0>, rot=<0 0 0 1>}
 jaw_left_pose {type=Curve, loc=<0 0 0>, rot=<0 0 0 1>}
 eye_convergence_pose {type=Curve, loc=<0 0 0>, rot=<0 0 0 1>}
 L_upperLidDown_pose {type=Curve, loc=<0 0 0>, rot=<0 0 0 1>}
 R_upperLidDown_pose {type=Curve, loc=<0 0 0>, rot=<0 0 0 1>}
 jaw_up_pose {type=Curve, loc=<0 0 0>, rot=<0 0 0 1>}
 jaw_forward_pose {type=Curve, loc=<0 0 0>, rot=<0 0 0 1>}
 jaw_back_pose {type=Curve, loc=<0 0 0>, rot=<0 0 0 1>}
 tongue_forward_pose {type=Curve, loc=<0 0 0>, rot=<0 0 0 1>}
 tongue_back_pose {type=Curve, loc=<0 0 0>, rot=<0 0 0 1>}
 tongue_up_pose {type=Curve, loc=<0 0 0>, rot=<0 0 0 1>}
 tongue_down_pose {type=Curve, loc=<0 0 0>, rot=<0 0 0 1>}
 base_pose {type=Curve, loc=<0 0 0>, rot=<0 0 0 1>}
 Hair_Scarf {type=Curve, loc=<0 0 0>, rot=<0 0 0 1>}
 Hair_Hat {type=Curve, loc=<0 0 0>, rot=<0 0 0 1>}
 Hair_Helmet {type=Curve, loc=<0 0 0>, rot=<0 0 0 1>}
 Hair_Glasses {type=Curve, loc=<0 0 0>, rot=<0 0 0 1>}
 Hair_Cap {type=Curve, loc=<0 0 0>, rot=<0 0 0 1>}
 Hair_Headsets {type=Curve, loc=<0 0 0>, rot=<0 0 0 1>}
 Hair_Hood {type=Curve, loc=<0 0 0>, rot=<0 0 0 1>}
 Face_Mask {type=Curve, loc=<0 0 0>, rot=<0 0 0 1>}
 HairStrands_Hat {type=Curve, loc=<0 0 0>, rot=<0 0 0 1>}
 Hat_Hair {type=Curve, loc=<0 0 0>, rot=<0 0 0 1>}
 Helmet_Hair {type=Curve, loc=<0 0 0>, rot=<0 0 0 1>}
 Hair_Muff {type=Curve, loc=<0 0 0>, rot=<0 0 0 1>}
 Hair_SpaceSuit {type=Curve, loc=<0 0 0>, rot=<0 0 0 1>}
 Hair_Hide {type=Curve, loc=<0 0 0>, rot=<0 0 0 1>}
 Head_SpaceSuit {type=Curve, loc=<0 0 0>, rot=<0 0 0 1>}
 Hair_Null {type=Curve, loc=<0 0 0>, rot=<0 0 0 1>}
 DisableFaceOverride {type=Curve, loc=<0 0 0>, rot=<0 0 0 1>}
 LiveLink {type=Curve, loc=<0 0 0>, rot=<0 0 0 1>}
 Hat_MESH {type=Curve, loc=<0 0 0>, rot=<0 0 0 1>}
 Null_MESH {type=Curve, loc=<0 0 0>, rot=<0 0 0 1>}
 Pose_0 {type=Curve, loc=<0 0 0>, rot=<0 0 0 1>}
 Hair_Mask {type=Curve, loc=<0 0 0>, rot=<0 0 0 1>}
 eye_convergence {type=Curve, loc=<0 0 0>, rot=<0 0 0 1>}
 headTransform {type=Curve, loc=<0 0 0>, rot=<0 0 0 1>}
 Pose_1 {type=Curve, loc=<0 0 0>, rot=<0 0 0 1>}
 Pose_2 {type=Curve, loc=<0 0 0>, rot=<0 0 0 1>}
 Pose_3 {type=Curve, loc=<0 0 0>, rot=<0 0 0 1>}
 Pose_4 {type=Curve, loc=<0 0 0>, rot=<0 0 0 1>}
 Pose_5 {type=Curve, loc=<0 0 0>, rot=<0 0 0 1>}
 Pose_6 {type=Curve, loc=<0 0 0>, rot=<0 0 0 1>}
 Pose_7 {type=Curve, loc=<0 0 0>, rot=<0 0 0 1>}
 Pose_8 {type=Curve, loc=<0 0 0>, rot=<0 0 0 1>}
 Pose_9 {type=Curve, loc=<0 0 0>, rot=<0 0 0 1>}
 Pose_10 {type=Curve, loc=<0 0 0>, rot=<0 0 0 1>}
 Pose_11 {type=Curve, loc=<0 0 0>, rot=<0 0 0 1>}
 Pose_12 {type=Curve, loc=<0 0 0>, rot=<0 0 0 1>}
 Pose_13 {type=Curve, loc=<0 0 0>, rot=<0 0 0 1>}
 Pose_14 {type=Curve, loc=<0 0 0>, rot=<0 0 0 1>}
 Pose_15 {type=Curve, loc=<0 0 0>, rot=<0 0 0 1>}
 Pose_16 {type=Curve, loc=<0 0 0>, rot=<0 0 0 1>}
 Pose_17 {type=Curve, loc=<0 0 0>, rot=<0 0 0 1>}
 Pose_18 {type=Curve, loc=<0 0 0>, rot=<0 0 0 1>}
 Pose_19 {type=Curve, loc=<0 0 0>, rot=<0 0 0 1>}
 Pose_20 {type=Curve, loc=<0 0 0>, rot=<0 0 0 1>}
 Pose_21 {type=Curve, loc=<0 0 0>, rot=<0 0 0 1>}
 Pose_22 {type=Curve, loc=<0 0 0>, rot=<0 0 0 1>}
 Pose_23 {type=Curve, loc=<0 0 0>, rot=<0 0 0 1>}
 Pose_24 {type=Curve, loc=<0 0 0>, rot=<0 0 0 1>}
 Pose_25 {type=Curve, loc=<0 0 0>, rot=<0 0 0 1>}
 Pose_26 {type=Curve, loc=<0 0 0>, rot=<0 0 0 1>}
 Pose_27 {type=Curve, loc=<0 0 0>, rot=<0 0 0 1>}
 Pose_28 {type=Curve, loc=<0 0 0>, rot=<0 0 0 1>}
 Pose_29 {type=Curve, loc=<0 0 0>, rot=<0 0 0 1>}
 Pose_30 {type=Curve, loc=<0 0 0>, rot=<0 0 0 1>}
 Pose_31 {type=Curve, loc=<0 0 0>, rot=<0 0 0 1>}
 Pose_32 {type=Curve, loc=<0 0 0>, rot=<0 0 0 1>}
 Pose_33 {type=Curve, loc=<0 0 0>, rot=<0 0 0 1>}
 Pose_34 {type=Curve, loc=<0 0 0>, rot=<0 0 0 1>}
 Pose_35 {type=Curve, loc=<0 0 0>, rot=<0 0 0 1>}
 Pose_36 {type=Curve, loc=<0 0 0>, rot=<0 0 0 1>}
 Pose_37 {type=Curve, loc=<0 0 0>, rot=<0 0 0 1>}
 Pose_38 {type=Curve, loc=<0 0 0>, rot=<0 0 0 1>}
 Pose_39 {type=Curve, loc=<0 0 0>, rot=<0 0 0 1>}
 Pose_40 {type=Curve, loc=<0 0 0>, rot=<0 0 0 1>}
 Pose_41 {type=Curve, loc=<0 0 0>, rot=<0 0 0 1>}
 Pose_42 {type=Curve, loc=<0 0 0>, rot=<0 0 0 1>}
 Pose_43 {type=Curve, loc=<0 0 0>, rot=<0 0 0 1>}
 Pose_44 {type=Curve, loc=<0 0 0>, rot=<0 0 0 1>}
 Pose_45 {type=Curve, loc=<0 0 0>, rot=<0 0 0 1>}
 Pose_46 {type=Curve, loc=<0 0 0>, rot=<0 0 0 1>}
 Pose_47 {type=Curve, loc=<0 0 0>, rot=<0 0 0 1>}
 Pose_48 {type=Curve, loc=<0 0 0>, rot=<0 0 0 1>}
 Pose_49 {type=Curve, loc=<0 0 0>, rot=<0 0 0 1>}
 Pose_50 {type=Curve, loc=<0 0 0>, rot=<0 0 0 1>}
 Pose_51 {type=Curve, loc=<0 0 0>, rot=<0 0 0 1>}
 Pose_52 {type=Curve, loc=<0 0 0>, rot=<0 0 0 1>}
 Pose_53 {type=Curve, loc=<0 0 0>, rot=<0 0 0 1>}
 Pose_54 {type=Curve, loc=<0 0 0>, rot=<0 0 0 1>}
 Pose_55 {type=Curve, loc=<0 0 0>, rot=<0 0 0 1>}
 Pose_56 {type=Curve, loc=<0 0 0>, rot=<0 0 0 1>}
 Pose_57 {type=Curve, loc=<0 0 0>, rot=<0 0 0 1>}
 Pose_58 {type=Curve, loc=<0 0 0>, rot=<0 0 0 1>}
 Pose_59 {type=Curve, loc=<0 0 0>, rot=<0 0 0 1>}
 Pose_60 {type=Curve, loc=<0 0 0>, rot=<0 0 0 1>}
 Pose_61 {type=Curve, loc=<0 0 0>, rot=<0 0 0 1>}
 Pose_62 {type=Curve, loc=<0 0 0>, rot=<0 0 0 1>}
 Pose_63 {type=Curve, loc=<0 0 0>, rot=<0 0 0 1>}
 Pose_64 {type=Curve, loc=<0 0 0>, rot=<0 0 0 1>}
 Pose_65 {type=Curve, loc=<0 0 0>, rot=<0 0 0 1>}
 Pose_66 {type=Curve, loc=<0 0 0>, rot=<0 0 0 1>}
 Pose_67 {type=Curve, loc=<0 0 0>, rot=<0 0 0 1>}
 Pose_68 {type=Curve, loc=<0 0 0>, rot=<0 0 0 1>}
 Pose_69 {type=Curve, loc=<0 0 0>, rot=<0 0 0 1>}
 eyeLookDownRight {type=Curve, loc=<0 0 0>, rot=<0 0 0 1>}
 eyeLookOutRight {type=Curve, loc=<0 0 0>, rot=<0 0 0 1>}
 eyeLookUpRight {type=Curve, loc=<0 0 0>, rot=<0 0 0 1>}
 eyeLookOutLeft {type=Curve, loc=<0 0 0>, rot=<0 0 0 1>}
 eyeBlinkLeft {type=Curve, loc=<0 0 0>, rot=<0 0 0 1>}
 eyeBlinkRight {type=Curve, loc=<0 0 0>, rot=<0 0 0 1>}
 eyeSquintLeft {type=Curve, loc=<0 0 0>, rot=<0 0 0 1>}
 eyeSquintRight {type=Curve, loc=<0 0 0>, rot=<0 0 0 1>}
 eyeWideLeft {type=Curve, loc=<0 0 0>, rot=<0 0 0 1>}
 eyeWideRight {type=Curve, loc=<0 0 0>, rot=<0 0 0 1>}
 jawForward {type=Curve, loc=<0 0 0>, rot=<0 0 0 1>}
 jawLeft {type=Curve, loc=<0 0 0>, rot=<0 0 0 1>}
 jawRight {type=Curve, loc=<0 0 0>, rot=<0 0 0 1>}
 jawOpen {type=Curve, loc=<0 0 0>, rot=<0 0 0 1>}
 mouthClose {type=Curve, loc=<0 0 0>, rot=<0 0 0 1>}
 mouthFunnel {type=Curve, loc=<0 0 0>, rot=<0 0 0 1>}
 mouthPucker {type=Curve, loc=<0 0 0>, rot=<0 0 0 1>}
 mouthLeft {type=Curve, loc=<0 0 0>, rot=<0 0 0 1>}
 mouthRight {type=Curve, loc=<0 0 0>, rot=<0 0 0 1>}
 mouthSmileLeft {type=Curve, loc=<0 0 0>, rot=<0 0 0 1>}
 mouthSmileRight {type=Curve, loc=<0 0 0>, rot=<0 0 0 1>}
 mouthFrownLeft {type=Curve, loc=<0 0 0>, rot=<0 0 0 1>}
 mouthFrownRight {type=Curve, loc=<0 0 0>, rot=<0 0 0 1>}
 mouthDimpleLeft {type=Curve, loc=<0 0 0>, rot=<0 0 0 1>}
 mouthDimpleRight {type=Curve, loc=<0 0 0>, rot=<0 0 0 1>}
 mouthStretchLeft {type=Curve, loc=<0 0 0>, rot=<0 0 0 1>}
 mouthStretchRight {type=Curve, loc=<0 0 0>, rot=<0 0 0 1>}
 mouthRollLower {type=Curve, loc=<0 0 0>, rot=<0 0 0 1>}
 mouthRollUpper {type=Curve, loc=<0 0 0>, rot=<0 0 0 1>}
 mouthShrugLower {type=Curve, loc=<0 0 0>, rot=<0 0 0 1>}
 mouthShrugUpper {type=Curve, loc=<0 0 0>, rot=<0 0 0 1>}
 mouthPressLeft {type=Curve, loc=<0 0 0>, rot=<0 0 0 1>}
 mouthPressRight {type=Curve, loc=<0 0 0>, rot=<0 0 0 1>}
 mouthLowerDownLeft {type=Curve, loc=<0 0 0>, rot=<0 0 0 1>}
 mouthLowerDownRight {type=Curve, loc=<0 0 0>, rot=<0 0 0 1>}
 mouthUpperUpLeft {type=Curve, loc=<0 0 0>, rot=<0 0 0 1>}
 mouthUpperUpRight {type=Curve, loc=<0 0 0>, rot=<0 0 0 1>}
 browDownLeft {type=Curve, loc=<0 0 0>, rot=<0 0 0 1>}
 browDownRight {type=Curve, loc=<0 0 0>, rot=<0 0 0 1>}
 browInnerUp {type=Curve, loc=<0 0 0>, rot=<0 0 0 1>}
 browOuterUpLeft {type=Curve, loc=<0 0 0>, rot=<0 0 0 1>}
 browOuterUpRight {type=Curve, loc=<0 0 0>, rot=<0 0 0 1>}
 cheekPuff {type=Curve, loc=<0 0 0>, rot=<0 0 0 1>}
 cheekSquintLeft {type=Curve, loc=<0 0 0>, rot=<0 0 0 1>}
 cheekSquintRight {type=Curve, loc=<0 0 0>, rot=<0 0 0 1>}
 noseSneerLeft {type=Curve, loc=<0 0 0>, rot=<0 0 0 1>}
 noseSneerRight {type=Curve, loc=<0 0 0>, rot=<0 0 0 1>}
 C_glabella_down {type=Curve, loc=<0 0 0>, rot=<0 0 0 1>}
 L_brow_squeeze_in {type=Curve, loc=<0 0 0>, rot=<0 0 0 1>}
 R_brow_squeeze_in {type=Curve, loc=<0 0 0>, rot=<0 0 0 1>}
 L_upper_lip_lower {type=Curve, loc=<0 0 0>, rot=<0 0 0 1>}
 R_upper_lip_lower {type=Curve, loc=<0 0 0>, rot=<0 0 0 1>}
 L_lower_lip_up {type=Curve, loc=<0 0 0>, rot=<0 0 0 1>}
 R_lower_lip_up {type=Curve, loc=<0 0 0>, rot=<0 0 0 1>}
 L_lip_corner_narrow {type=Curve, loc=<0 0 0>, rot=<0 0 0 1>}
 R_lip_corner_narrow {type=Curve, loc=<0 0 0>, rot=<0 0 0 1>}
 phoneme_ch {type=Curve, loc=<0 0 0>, rot=<0 0 0 1>}
 phoneme_mbp {type=Curve, loc=<0 0 0>, rot=<0 0 0 1>}
 phoneme_fv {type=Curve, loc=<0 0 0>, rot=<0 0 0 1>}
 mouth_up {type=Curve, loc=<0 0 0>, rot=<0 0 0 1>}
 mouth_down {type=Curve, loc=<0 0 0>, rot=<0 0 0 1>}
 L_cheek_up {type=Curve, loc=<0 0 0>, rot=<0 0 0 1>}
 R_cheek_up {type=Curve, loc=<0 0 0>, rot=<0 0 0 1>}
 L_upperLidDown {type=Curve, loc=<0 0 0>, rot=<0 0 0 1>}
 R_upperLidDown {type=Curve, loc=<0 0 0>, rot=<0 0 0 1>}
 jaw_up {type=Curve, loc=<0 0 0>, rot=<0 0 0 1>}
 jaw_back {type=Curve, loc=<0 0 0>, rot=<0 0 0 1>}
 tongue_forward {type=Curve, loc=<0 0 0>, rot=<0 0 0 1>}
 tongue_back {type=Curve, loc=<0 0 0>, rot=<0 0 0 1>}
 tongue_up {type=Curve, loc=<0 0 0>, rot=<0 0 0 1>}
 tongue_down {type=Curve, loc=<0 0 0>, rot=<0 0 0 1>}
 basePose {type=Curve, loc=<0 0 0>, rot=<0 0 0 1>}
 neutral {type=Curve, loc=<0 0 0>, rot=<0 0 0 1>}
 cheek_puff_low {type=Curve, loc=<0 0 0>, rot=<0 0 0 1>}
 Desert_Island_Jonesy_Mouth {type=Curve, loc=<0 0 0>, rot=<0 0 0 1>}
 HT_L1_TransitionCurve {type=Curve, loc=<0 0 0>, rot=<0 0 0 1>}
 Hair_CapShape {type=Curve, loc=<0 0 0>, rot=<0 0 0 1>}
 Hair_CapShape1 {type=Curve, loc=<0 0 0>, rot=<0 0 0 1>}
 jekyll {type=Curve, loc=<0 0 0>, rot=<0 0 0 1>}
 JH_TransformCurve {type=Curve, loc=<0 0 0>, rot=<0 0 0 1>}
 eyes_LOD0 {type=Curve, loc=<0 0 0>, rot=<0 0 0 1>}
 MaterialCurveA {type=Curve, loc=<0 0 0>, rot=<0 0 0 1>}
 MaterialCurveB {type=Curve, loc=<0 0 0>, rot=<0 0 0 1>}
 MaterialCurveC {type=Curve, loc=<0 0 0>, rot=<0 0 0 1>}
 MaterialCurveD {type=Curve, loc=<0 0 0>, rot=<0 0 0 1>}
 MaterialCurveE {type=Curve, loc=<0 0 0>, rot=<0 0 0 1>}
 Custom_Frontend_Face_Idle {type=Curve, loc=<0 0 0>, rot=<0 0 0 1>}
 blendParent1 {type=Curve, loc=<0 0 0>, rot=<0 0 0 1>}
 BaseHeadDefaultFaceIdle {type=Curve, loc=<0 0 0>, rot=<0 0 0 1>}
 VeiledTransitionGlowAmount {type=Curve, loc=<0 0 0>, rot=<0 0 0 1>}
 is_3L {type=Curve, loc=<0 0 0>, rot=<0 0 0 1>}
 Hair_Hat_LOD0 {type=Curve, loc=<0 0 0>, rot=<0 0 0 1>}
 Hair_Helmet_LOD0 {type=Curve, loc=<0 0 0>, rot=<0 0 0 1>}
 Hair_Null_LOD0 {type=Curve, loc=<0 0 0>, rot=<0 0 0 1>}
 mouth_close_pose {type=Curve, loc=<0 0 0>, rot=<0 0 0 1>}
 mouth_pucker_pose {type=Curve, loc=<0 0 0>, rot=<0 0 0 1>}
 mouth_left_pose {type=Curve, loc=<0 0 0>, rot=<0 0 0 1>}
 mouth_right_pose {type=Curve, loc=<0 0 0>, rot=<0 0 0 1>}
 L_mouth_stretch_pose {type=Curve, loc=<0 0 0>, rot=<0 0 0 1>}
 R_mouth_stretch_pose {type=Curve, loc=<0 0 0>, rot=<0 0 0 1>}
 mouth_roll_lower_pose {type=Curve, loc=<0 0 0>, rot=<0 0 0 1>}
 mouth_roll_upper_pose {type=Curve, loc=<0 0 0>, rot=<0 0 0 1>}
 mouth_shrug_lower_pose {type=Curve, loc=<0 0 0>, rot=<0 0 0 1>}
 mouth_shrug_upper_pose {type=Curve, loc=<0 0 0>, rot=<0 0 0 1>}
 L_mouth_press_pose {type=Curve, loc=<0 0 0>, rot=<0 0 0 1>}
 R_mouth_press_pose {type=Curve, loc=<0 0 0>, rot=<0 0 0 1>}
 cheek_puff_pose {type=Curve, loc=<0 0 0>, rot=<0 0 0 1>}
 L_nose_wrinkler_pose {type=Curve, loc=<0 0 0>, rot=<0 0 0 1>}
 R_nose_wrinkler_pose {type=Curve, loc=<0 0 0>, rot=<0 0 0 1>}
 base {type=Curve, loc=<0 0 0>, rot=<0 0 0 1>}
 eyeLookDownLeft {type=Curve, loc=<0 0 0>, rot=<0 0 0 1>}
 eyeLookInLeft {type=Curve, loc=<0 0 0>, rot=<0 0 0 1>}
 eyeLookUpLeft {type=Curve, loc=<0 0 0>, rot=<0 0 0 1>}
 eyeLookInRight {type=Curve, loc=<0 0 0>, rot=<0 0 0 1>}
 is_face_bone_anim {type=Curve, loc=<0 0 0>, rot=<0 0 0 1>}
 FullHourglass {type=Curve, loc=<0 0 0>, rot=<0 0 0 1>}
 LeanRectangle {type=Curve, loc=<0 0 0>, rot=<0 0 0 1>}
 InvertedTriangle {type=Curve, loc=<0 0 0>, rot=<0 0 0 1>}
 L_eye {type=Curve, loc=<0 0 0>, rot=<0 0 0 1>}
 R_eye {type=Curve, loc=<0 0 0>, rot=<0 0 0 1>}
 head {type=Curve, loc=<0 0 0>, rot=<0 0 0 1>}
 dyn_ear_l_master {type=Curve, loc=<0 0 0>, rot=<0 0 0 1>}
 root {type=Curve, loc=<0 0 0>, rot=<0 0 0 1>}
 dyn_ear02_l {type=Curve, loc=<0 0 0>, rot=<0 0 0 1>}
 dyn_ear03_l {type=Curve, loc=<0 0 0>, rot=<0 0 0 1>}
 dyn_ear04_l {type=Curve, loc=<0 0 0>, rot=<0 0 0 1>}
 dyn_ear05_l {type=Curve, loc=<0 0 0>, rot=<0 0 0 1>}
 dyn_ear01_l {type=Curve, loc=<0 0 0>, rot=<0 0 0 1>}
 flipbook_face_L_eye_index {type=Curve, loc=<0 0 0>, rot=<0 0 0 1>}
 flipbook_face_R_eye_index {type=Curve, loc=<0 0 0>, rot=<0 0 0 1>}
 flipbook_face_L_brow_index {type=Curve, loc=<0 0 0>, rot=<0 0 0 1>}
 flipbook_face_R_brow_index {type=Curve, loc=<0 0 0>, rot=<0 0 0 1>}
 flipbook_face_mouth_index {type=Curve, loc=<0 0 0>, rot=<0 0 0 1>}
 flipbook_face_bypass_mapping {type=Curve, loc=<0 0 0>, rot=<0 0 0 1>}
 flipbook_face_is_idling {type=Curve, loc=<0 0 0>, rot=<0 0 0 1>}
]
variables [isEmoting {type=bool, default=False}]
