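# Physics asset for the HarborSnow large male player head cosmetic, defining secondary-motion bodies for its ears and hair.
# Left ear, right ear and hair are each a chain of three capsule bodies, with the first link of each constrained to a capsule on the head bone.
# Constraints use moderate swing limits, slightly tighter on the hair, and very narrow twist limits, so the chains can bend but barely roll.
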PhysicsAsset /BRCosmetics/Characters/Player/Male/Large/Heads/M_LRG_HarborSnow_Head/Meshes/M_LRG_HarborSnow_Head_Physics
bodies {
 dyn_ear_l_01 [{shapeType=Capsule, length=3.6914, radius=3.3086, shapeName=dyn_ear_l_01_capsule, center=<2 0 0>, rotation=<rot 90 0 0>}]
 dyn_ear_l_02 [{shapeType=Capsule, length=3.7529, radius=3.2471, shapeName=dyn_ear_l_02_capsule, center=<2 0 0>, rotation=<rot 90 0 0>}]
 dyn_ear_l_03 [{shapeType=Capsule, length=3.6794, radius=3.3206, shapeName=dyn_ear_l_03_capsule, center=<2 0 0>, rotation=<rot 90 0 0>}]
 dyn_ear_r_01 [{shapeType=Capsule, length=3.7141, radius=3.2859, shapeName=dyn_ear_r_01_capsule, center=<2 0 0>, rotation=<rot 90 0 0>}]
 dyn_ear_r_02 [{shapeType=Capsule, length=4, radius=3, shapeName=dyn_ear_r_02_capsule, center=<2 0 0>, rotation=<rot 90 0 0>}]
 dyn_ear_r_03 [{shapeType=Capsule, length=4, radius=3, shapeName=dyn_ear_r_03_capsule, center=<2 0 0>, rotation=<rot 90 0 0>}]
 dyn_hair_01 [{shapeType=Capsule, length=2.7157, radius=4.2843, shapeName=dyn_hair_01_capsule, center=<2 0 0>, rotation=<rot 90 0 0>}]
 dyn_hair_02 [{shapeType=Capsule, length=3.1588, radius=3.8412, shapeName=dyn_hair_02_capsule, center=<2 0 0>, rotation=<rot 90 0 0>}]
 dyn_hair_03 [{shapeType=Capsule, length=3.456, radius=3.544, shapeName=dyn_hair_03_capsule, center=<2 0 0>, rotation=<rot 90 0 0>}]
 head [{shapeType=Capsule, length=2.8082, radius=1.5957, shapeName=head_capsule}]
}
constraints [
 {bone1Name=dyn_ear_l_01, bone2Name=head, swing1LimitDegrees=25, swing2LimitDegrees=25, twistLimitDegrees=2}
 {bone1Name=dyn_ear_l_02, bone2Name=dyn_ear_l_01, swing1LimitDegrees=30, swing2LimitDegrees=30, twistLimitDegrees=2}
 {bone1Name=dyn_ear_l_03, bone2Name=dyn_ear_l_02, swing1LimitDegrees=30, swing2LimitDegrees=30, twistLimitDegrees=2}
 {bone1Name=dyn_ear_r_01, bone2Name=head, swing1LimitDegrees=25, swing2LimitDegrees=25, twistLimitDegrees=2}
 {bone1Name=dyn_ear_r_02, bone2Name=dyn_ear_r_01, swing1LimitDegrees=30, swing2LimitDegrees=30, twistLimitDegrees=2}
 {bone1Name=dyn_ear_r_03, bone2Name=dyn_ear_r_02, swing1LimitDegrees=30, swing2LimitDegrees=30, twistLimitDegrees=2}
 {bone1Name=dyn_hair_01, bone2Name=head, swing1LimitDegrees=20, swing2LimitDegrees=20, twistLimitDegrees=2}
 {bone1Name=dyn_hair_02, bone2Name=dyn_hair_01, swing1LimitDegrees=20, swing2LimitDegrees=20, twistLimitDegrees=2}
 {bone1Name=dyn_hair_03, bone2Name=dyn_hair_02, swing1LimitDegrees=20, swing2LimitDegrees=20, twistLimitDegrees=2}
]
